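# Physics asset for the medium female BugBeliever cosmetic body, defining capsule bodies and joints for its front and back chains and a hanging grenade.
# Each chain is six limited-swing, low-twist links from a base bone, with the front base locked to the pelvis.
# Free-rotation joints tie both chain tips to a mid anchor off the right thigh, closing a loop.
# The grenade hangs from the first front link.
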PhysicsAsset /BRCosmetics/Characters/Player/Female/Medium/Bodies/F_MED_BugBeliever/Meshes/F_MED_BugBeliever_Physics
bodies {
 Acncher_chain_Mid_01 [{shapeType=Capsule, length=1, radius=1, shapeName=Acncher_chain_Mid_01_capsule, rotation=<rot 0 -90 0>}]
 Acncher_chain_Mid_02 [{shapeType=Capsule, length=1, radius=1, shapeName=Acncher_chain_Mid_02_capsule, rotation=<rot 0 -90 0>}]
 Base_Granade [{shapeType=Capsule, length=0.505, radius=0.505, shapeName=Base_Granade_capsule, rotation=<rot 0 -90 0>}]
 Base_chain_Back [{shapeType=Capsule, length=0.505, radius=0.505, shapeName=Base_chain_Back_capsule, rotation=<rot 0 -90 0>}]
 Base_chain_Front [{shapeType=Capsule, length=0.5, radius=0.5, shapeName=Base_chain_Front_capsule}]
 Correct_chain_Front [{shapeType=Capsule, length=0.505, radius=0.505, shapeName=Correct_chain_Front_capsule, rotation=<rot 0 -90 0>}]
 Dyn_Granade_01 [{shapeType=Capsule, length=0.1558, radius=2.3505, shapeName=Dyn_Granade_01_capsule, center=<4.8226 0 0>, rotation=<rot 90 0 0>}]
 Dyn_chain_Back_01 [{
   shapeType=Capsule
   length=1.7092
   radius=1.7068
   shapeName=Dyn_chain_Back_01_capsule
   center=<1.2594 -0.0109 -0.1034>
   rotation=<rot 0 -103.7298 -88.3522>
  }]
 Dyn_chain_Back_02 [{shapeType=Capsule, length=1.5148, radius=1.6118, shapeName=Dyn_chain_Back_02_capsule, center=<3.0678 0 0>, rotation=<rot 90 -0.0002 0>}]
 Dyn_chain_Back_03 [{
   shapeType=Capsule
   length=3.4797
   radius=1.7107
   shapeName=Dyn_chain_Back_03_capsule
   center=<3.0678 0 0>
   rotation=<rot 81.5924 -0.0003 0.0004>
  }]
 Dyn_chain_Back_04 [{
   shapeType=Capsule
   length=5.339
   radius=1.7883
   shapeName=Dyn_chain_Back_04_capsule
   center=<3.1508 0 -0.0112>
   rotation=<rot 82.3308 0.0001 9.8387>
  }]
 Dyn_chain_Back_05 [{shapeType=Capsule, length=3.7165, radius=1.8541, shapeName=Dyn_chain_Back_05_capsule, center=<3.0678 0 0>, rotation=<rot 90 -10.2854 0>}]
 Dyn_chain_Back_06 [{shapeType=Capsule, length=0.8464, radius=1.8541, shapeName=Dyn_chain_Back_06_capsule, center=<0.1915 0 0>, rotation=<rot 90 0 0>}]
 Dyn_chain_Front_01 [{
   shapeType=Capsule
   length=2.6869
   radius=1.6916
   shapeName=Dyn_chain_Front_01_capsule
   center=<1.9704 0.5728 0.2102>
   rotation=<rot -79.0473 -163.8022 179.9999>
  }]
 Dyn_chain_Front_02 [{shapeType=Capsule, length=3.3882, radius=1.6118, shapeName=Dyn_chain_Front_02_capsule, center=<3.0678 0 0.6392>, rotation=<rot 90 -0.0008 0>}]
 Dyn_chain_Front_03 [{
   shapeType=Capsule
   length=3.237
   radius=1.763
   shapeName=Dyn_chain_Front_03_capsule
   center=<3.0678 0 0.1136>
   rotation=<rot 78.4063 -0.0007 0.0002>
  }]
 Dyn_chain_Front_04 [{
   shapeType=Capsule
   length=3.7165
   radius=1.8541
   shapeName=Dyn_chain_Front_04_capsule
   center=<3.0678 0 0>
   rotation=<rot 79.8241 -0.0006 0.0004>
  }]
 Dyn_chain_Front_05 [{
   shapeType=Capsule
   length=3.7165
   radius=1.8541
   shapeName=Dyn_chain_Front_05_capsule
   center=<3.0678 0 0>
   rotation=<rot 85.4964 -0.0008 0.0001>
  }]
 Dyn_chain_Front_06 [{shapeType=Capsule, length=0.8464, radius=1.8541, shapeName=Dyn_chain_Front_06_capsule, center=<0.1915 0 0>, rotation=<rot 90 0 0>}]
 pelvis [
  {shapeType=Capsule, length=7.1434, radius=8.8566, shapeName=pelvis_capsule, center=<0 -1.4062 0>}
  {shapeType=Capsule, length=12.6087, radius=8.1011, shapeName=pelvis_capsule, center=<-2.1858 -3.1508 0>}
 ]
 spine_01 [{shapeType=Capsule, length=3.5, radius=9, shapeName=spine_01_capsule}]
 thigh_r [{shapeType=Capsule, length=20.5674, radius=6.9326, shapeName=thigh_r_capsule, center=<12.7796 0.5477 -2.0044>, rotation=<rot 90 -3.5104 0>}]
}
constraints [
 {bone1Name=Dyn_chain_Back_02, bone2Name=Dyn_chain_Back_01, swing1LimitDegrees=20, swing2LimitDegrees=10, twistLimitDegrees=0}
 {bone1Name=Dyn_chain_Back_03, bone2Name=Dyn_chain_Back_02, swing1LimitDegrees=20, swing2LimitDegrees=20, twistLimitDegrees=5}
 {bone1Name=Dyn_chain_Back_04, bone2Name=Dyn_chain_Back_03, swing1LimitDegrees=20, swing2LimitDegrees=20, twistLimitDegrees=5}
 {bone1Name=Dyn_chain_Back_05, bone2Name=Dyn_chain_Back_04, swing1LimitDegrees=20, swing2LimitDegrees=20, twistLimitDegrees=5}
 {bone1Name=Dyn_chain_Back_06, bone2Name=Dyn_chain_Back_05, swing1LimitDegrees=25, swing2LimitDegrees=25, twistLimitDegrees=5}
 {bone1Name=Dyn_chain_Front_01, bone2Name=Correct_chain_Front, swing1LimitDegrees=10, swing2LimitDegrees=30, twistLimitDegrees=0}
 {bone1Name=Dyn_chain_Front_02, bone2Name=Dyn_chain_Front_01, swing1LimitDegrees=15, swing2LimitDegrees=30, twistLimitDegrees=0}
 {bone1Name=Dyn_chain_Front_03, bone2Name=Dyn_chain_Front_02, swing1LimitDegrees=25, swing2LimitDegrees=25, twistLimitDegrees=5}
 {bone1Name=Dyn_chain_Front_04, bone2Name=Dyn_chain_Front_03, swing1LimitDegrees=25, swing2LimitDegrees=25, twistLimitDegrees=5}
 {bone1Name=Dyn_chain_Front_05, bone2Name=Dyn_chain_Front_04, swing1LimitDegrees=25, swing2LimitDegrees=25, twistLimitDegrees=5}
 {bone1Name=Dyn_chain_Front_06, bone2Name=Dyn_chain_Front_05, swing1LimitDegrees=25, swing2LimitDegrees=25, twistLimitDegrees=5}
 {
  bone1Name=Base_chain_Front
  bone2Name=pelvis
  swing1LimitDegrees=45
  swing1Motion=Locked
  swing2LimitDegrees=45
  swing2Motion=Locked
  twistLimitDegrees=45
  twistMotion=Locked
 }
 {bone1Name=Acncher_chain_Mid_01, bone2Name=thigh_r, swing1LimitDegrees=30, swing2LimitDegrees=20, twistLimitDegrees=0}
 {bone1Name=Acncher_chain_Mid_02, bone2Name=Acncher_chain_Mid_01, swing1LimitDegrees=25, swing2LimitDegrees=25, twistLimitDegrees=5}
 {
  bone1Name=Dyn_chain_Back_06
  bone2Name=Dyn_chain_Front_06
  swing1LimitDegrees=45
  swing1Motion=Free
  swing2LimitDegrees=45
  swing2Motion=Free
  twistLimitDegrees=45
  twistMotion=Free
 }
 {
  bone1Name=Dyn_chain_Front_06
  bone2Name=Acncher_chain_Mid_02
  swing1LimitDegrees=45
  swing1Motion=Free
  swing2LimitDegrees=45
  swing2Motion=Free
  twistLimitDegrees=45
  twistMotion=Free
 }
 {
  bone1Name=Dyn_chain_Back_06
  bone2Name=Acncher_chain_Mid_02
  swing1LimitDegrees=45
  swing1Motion=Free
  swing2LimitDegrees=45
  swing2Motion=Free
  twistLimitDegrees=45
  twistMotion=Free
 }
 {bone1Name=Dyn_chain_Back_01, bone2Name=Base_chain_Back, swing1LimitDegrees=10, swing2LimitDegrees=10, twistLimitDegrees=0}
 {
  bone1Name=Base_Granade
  bone2Name=Dyn_chain_Front_01
  swing1LimitDegrees=45
  swing1Motion=Free
  swing2LimitDegrees=45
  swing2Motion=Free
  twistLimitDegrees=45
  twistMotion=Free
 }
 {bone1Name=Dyn_Granade_01, bone2Name=Base_Granade, swing1LimitDegrees=0, swing2LimitDegrees=50, twistLimitDegrees=0}
 {
  bone1Name=Correct_chain_Front
  bone2Name=Base_chain_Front
  swing1LimitDegrees=45
  swing1Motion=Free
  swing2LimitDegrees=45
  swing2Motion=Free
  twistLimitDegrees=45
  twistMotion=Free
 }
]
