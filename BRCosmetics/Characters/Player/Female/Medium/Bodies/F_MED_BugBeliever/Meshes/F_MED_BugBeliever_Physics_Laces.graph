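# Physics asset for the F_MED_BugBeliever female medium body cosmetic, covering a left and right pair of laces on the upper torso.
# Capsule bodies cover two upper spine bones, both clavicles and a base and dynamic bone per lace.
# Each dynamic lace hangs off its base with limited swing and zero twist, tuned slightly differently per side.
# The bases and clavicles attach to spine_05, which links to spine_04, all with free angular motion.
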PhysicsAsset /BRCosmetics/Characters/Player/Female/Medium/Bodies/F_MED_BugBeliever/Meshes/F_MED_BugBeliever_Physics_Laces
bodies {
 Base_laces_L [{shapeType=Capsule, length=0.5, radius=0.5, shapeName=Base_laces_L_capsule}]
 Base_laces_R [{shapeType=Capsule, length=0.5, radius=0.5, shapeName=Base_laces_R_capsule}]
 Dyn_laces_L [{shapeType=Capsule, length=2.9316, radius=1.5077, shapeName=Dyn_laces_L_capsule, center=<6.6665 0 0.4268>, rotation=<rot 90 0 0>}]
 Dyn_laces_R [{shapeType=Capsule, length=2.1768, radius=2.2625, shapeName=Dyn_laces_R_capsule, center=<6.6665 0 0.4268>, rotation=<rot 90 0 0>}]
 clavicle_l [{shapeType=Capsule, length=3.5183, radius=5.7514, shapeName=clavicle_l_capsule, center=<2.3898 0 0>, rotation=<rot 90 0 0>}]
 clavicle_r [{shapeType=Capsule, length=3.5183, radius=5.7514, shapeName=clavicle_r_capsule, center=<-2.3898 0 0>, rotation=<rot 90 180 0>}]
 spine_04 [{shapeType=Capsule, length=12.6214, radius=10.6173, shapeName=spine_04_capsule}]
 spine_05 [{shapeType=Capsule, length=7.0662, radius=7.8202, shapeName=spine_05_capsule}]
}
constraints [
 {
  bone1Name=clavicle_l
  bone2Name=spine_05
  swing1LimitDegrees=45
  swing1Motion=Free
  swing2LimitDegrees=45
  swing2Motion=Free
  twistLimitDegrees=45
  twistMotion=Free
 }
 {
  bone1Name=clavicle_r
  bone2Name=spine_05
  swing1LimitDegrees=45
  swing1Motion=Free
  swing2LimitDegrees=45
  swing2Motion=Free
  twistLimitDegrees=45
  twistMotion=Free
 }
 {
  bone1Name=Base_laces_L
  bone2Name=spine_05
  swing1LimitDegrees=45
  swing1Motion=Free
  swing2LimitDegrees=45
  swing2Motion=Free
  twistLimitDegrees=45
  twistMotion=Free
 }
 {bone1Name=Dyn_laces_L, bone2Name=Base_laces_L, swing1LimitDegrees=20, swing2LimitDegrees=40, twistLimitDegrees=0}
 {
  bone1Name=Base_laces_R
  bone2Name=spine_05
  swing1LimitDegrees=45
  swing1Motion=Free
  swing2LimitDegrees=45
  swing2Motion=Free
  twistLimitDegrees=45
  twistMotion=Free
 }
 {bone1Name=Dyn_laces_R, bone2Name=Base_laces_R, swing1LimitDegrees=13, swing2LimitDegrees=40, twistLimitDegrees=0}
 {
  bone1Name=spine_05
  bone2Name=spine_04
  swing1LimitDegrees=45
  swing1Motion=Free
  swing2LimitDegrees=45
  swing2Motion=Free
  twistLimitDegrees=45
  twistMotion=Free
 }
]
