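# Physics asset for the face accessory mesh of the female medium BugBeliever character, defining bodies and joints for a side hair piece.
# A capsule chain runs from the hair base and forks after the second dynamic bone into two two-bone branches, with swing limits tightening toward the tips and narrow twist throughout.
# Capsules on the head, right clavicle and right upper arm supply body volumes, while the hair-to-head and arm-to-clavicle constraints leave rotation free.
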PhysicsAsset /BRCosmetics/Characters/Player/Female/Medium/Bodies/F_MED_BugBeliever/Meshes/Parts/F_MED_BugBeliever_FaceAcc_Physics
bodies {
 Base_Hair_Side [{shapeType=Capsule, length=0.5, radius=0.5, shapeName=Base_Hair_Side_capsule}]
 Dyn_Hair_Side_01 [{shapeType=Capsule, length=2.709, radius=2.291, shapeName=Dyn_Hair_Side_01_capsule, center=<2.3694 0 1.213>, rotation=<rot -90 90 0>}]
 Dyn_Hair_Side_02 [{shapeType=Capsule, length=2.4069, radius=1.7871, shapeName=Dyn_Hair_Side_02_capsule, center=<2.4468 0 0.7>, rotation=<rot -90 90 0>}]
 Dyn_Hair_Side_03 [{shapeType=Capsule, length=2.5995, radius=1.6815, shapeName=Dyn_Hair_Side_03_capsule, center=<2.2343 0 0.6004>, rotation=<rot -90 -0.0005 0>}]
 Dyn_Hair_Side_04 [{shapeType=Capsule, length=3.5, radius=1.5, shapeName=Dyn_Hair_Side_04_capsule, center=<2.7891 0 0.6004>, rotation=<rot -90 -0.0006 0>}]
 Dyn_Hair_Side_05 [{
   shapeType=Capsule
   length=2.989
   radius=1.511
   shapeName=Dyn_Hair_Side_05_capsule
   center=<1.9064 0.2189 0.6004>
   rotation=<rot -90 -0.0005 0>
  }]
 Dyn_Hair_Side_06 [{shapeType=Capsule, length=3.5, radius=1.5, shapeName=Dyn_Hair_Side_06_capsule, center=<2.4403 0 0.6004>, rotation=<rot -90 -0.0006 0>}]
 clavicle_r [{shapeType=Capsule, length=5, radius=5, shapeName=clavicle_r_capsule, center=<-5.524 0 0>, rotation=<rot 90 0 0>}]
 head [
  {shapeType=Capsule, length=3.4351, radius=5.865, shapeName=head_capsule, center=<1.3983 2.7138 0>, rotation=<rot 90 0 0>}
  {shapeType=Capsule, length=5.5006, radius=3.2547, shapeName=head_capsule, center=<0 0 3.6658>, rotation=<rot -90 -0.0002 0>}
 ]
 upperarm_r [{shapeType=Capsule, length=17.9131, radius=4.8058, shapeName=upperarm_r_capsule, center=<-12.0338 0 0>, rotation=<rot 90 0 0>}]
}
constraints [
 {
  bone1Name=Base_Hair_Side
  bone2Name=head
  swing1LimitDegrees=45
  swing1Motion=Free
  swing2LimitDegrees=45
  swing2Motion=Free
  twistLimitDegrees=45
  twistMotion=Free
 }
 {bone1Name=Dyn_Hair_Side_01, bone2Name=Base_Hair_Side, swing1LimitDegrees=20, swing2LimitDegrees=20, twistLimitDegrees=5}
 {bone1Name=Dyn_Hair_Side_02, bone2Name=Dyn_Hair_Side_01, swing1LimitDegrees=15, swing2LimitDegrees=15, twistLimitDegrees=5}
 {bone1Name=Dyn_Hair_Side_03, bone2Name=Dyn_Hair_Side_02, swing1LimitDegrees=10, swing2LimitDegrees=10, twistLimitDegrees=5}
 {bone1Name=Dyn_Hair_Side_04, bone2Name=Dyn_Hair_Side_03, swing1LimitDegrees=10, swing2LimitDegrees=10, twistLimitDegrees=5}
 {bone1Name=Dyn_Hair_Side_05, bone2Name=Dyn_Hair_Side_02, swing1LimitDegrees=10, swing2LimitDegrees=10, twistLimitDegrees=5}
 {bone1Name=Dyn_Hair_Side_06, bone2Name=Dyn_Hair_Side_05, swing1LimitDegrees=10, swing2LimitDegrees=10, twistLimitDegrees=5}
 {
  bone1Name=upperarm_r
  bone2Name=clavicle_r
  swing1LimitDegrees=45
  swing1Motion=Free
  swing2LimitDegrees=45
  swing2Motion=Free
  twistLimitDegrees=45
  twistMotion=Free
 }
]
