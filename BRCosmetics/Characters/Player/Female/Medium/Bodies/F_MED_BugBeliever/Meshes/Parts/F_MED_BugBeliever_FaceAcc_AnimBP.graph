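# Post-copy AnimBP for the BugBeliever female face accessory mesh, adding secondary motion to hair and earrings.
# It layers the Body part's pose, cached on customization load, over the attached parent's pose from the clavicles.
# A head-space RigidBody simulation then feeds chained, spring-driven AnimDynamics for back hair, front hair and earrings.
# A separate side-hair AnimDynamics chain, repeating bone 04, is not connected to the output.
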AnimBlueprint /BRCosmetics/Characters/Player/Female/Medium/Bodies/F_MED_BugBeliever/Meshes/Parts/F_MED_BugBeliever_FaceAcc_AnimBP
target_skeleton=&/F_MED_BugBeliever_FaceAcc_Skeleton
graphs [
 AnimGraph {nodes [
   n1 = Anim Dynamics - Bone: Dyn_Hair_Side_01(Alpha=1, ExternalForce="0,0,0") {
    angularDampingOverride=0.9999
    angularSpringConstant=2000
    bAngularSpring=true
    bOverrideAngularDamping=true
    bUseGravityOverride=true
    simulationSpace=RootRelative
    alphaBoolBlend {blendOption=Linear}
    boundBone {boneName=Dyn_Hair_Side_01}
    componentAppliedLinearAccClamp=<500 500 500>
    componentLinearAccScale=<0.7 0.7 0.7>
    componentLinearVelScale=<0.7 0.7 0.7>
    gravityOverride {z=-980}
    physicsBodyDefinitions [{
      boundBone {boneName=Dyn_Hair_Side_01}
      boxExtents=<10 10 10>
      constraintSetup {angularLimitsMax=<0 0 7>, angularLimitsMin=<0 -20 -7>, angularTarget=<0 0 0>}
      localJointOffset=<0.5 0 0>
     }]
   }
   n2 = Anim Dynamics - Bone: Dyn_Hair_Side_02(ComponentPose=@n1, Alpha=1, ExternalForce="0,0,0") {
    angularDampingOverride=0.9999
    angularSpringConstant=2000
    bAngularSpring=true
    bOverrideAngularDamping=true
    bUseGravityOverride=true
    simulationSpace=RootRelative
    alphaBoolBlend {blendOption=Linear}
    boundBone {boneName=Dyn_Hair_Side_02}
    componentAppliedLinearAccClamp=<500 500 500>
    componentLinearAccScale=<0.7 0.7 0.7>
    componentLinearVelScale=<0.7 0.7 0.7>
    gravityOverride {z=-980}
    physicsBodyDefinitions [{
      boundBone {boneName=Dyn_Hair_Side_02}
      boxExtents=<10 10 10>
      constraintSetup {angularLimitsMax=<0 0 7>, angularLimitsMin=<0 -20 -7>, angularTarget=<0 0 0>}
      localJointOffset=<0.5 0 0>
     }]
   }
   n3 = Anim Dynamics - Bone: Dyn_Hair_Side_03(ComponentPose=@n2, Alpha=1, ExternalForce="0,0,0") {
    angularDampingOverride=0.9999
    angularSpringConstant=2000
    bAngularSpring=true
    bOverrideAngularDamping=true
    bUseGravityOverride=true
    simulationSpace=RootRelative
    alphaBoolBlend {blendOption=Linear}
    boundBone {boneName=Dyn_Hair_Side_03}
    componentAppliedLinearAccClamp=<500 500 500>
    componentLinearAccScale=<0.7 0.7 0.7>
    componentLinearVelScale=<0.7 0.7 0.7>
    gravityOverride {z=-980}
    physicsBodyDefinitions [{
      boundBone {boneName=Dyn_Hair_Side_03}
      boxExtents=<10 10 10>
      constraintSetup {angularLimitsMax=<0 0 7>, angularLimitsMin=<0 -20 -7>, angularTarget=<0 0 0>}
      localJointOffset=<0.5 0 0>
     }]
   }
   n4 = Anim Dynamics - Bone: Dyn_Hair_Side_04(ComponentPose=@n3, Alpha=1, ExternalForce="0,0,0") {
    angularDampingOverride=0.9999
    angularSpringConstant=2000
    bAngularSpring=true
    bOverrideAngularDamping=true
    bUseGravityOverride=true
    simulationSpace=RootRelative
    alphaBoolBlend {blendOption=Linear}
    boundBone {boneName=Dyn_Hair_Side_04}
    componentAppliedLinearAccClamp=<500 500 500>
    componentLinearAccScale=<0.7 0.7 0.7>
    componentLinearVelScale=<0.7 0.7 0.7>
    gravityOverride {z=-980}
    physicsBodyDefinitions [{
      boundBone {boneName=Dyn_Hair_Side_04}
      boxExtents=<10 10 10>
      constraintSetup {angularLimitsMax=<0 0 7>, angularLimitsMin=<0 -20 -7>, angularTarget=<0 0 0>}
      localJointOffset=<0.5 0 0>
     }]
   }
   n5 = Anim Dynamics - Bone: Dyn_Hair_Side_05(ComponentPose=@n4, Alpha=1, ExternalForce="0,0,0") {
    angularDampingOverride=0.9999
    angularSpringConstant=2000
    bAngularSpring=true
    bOverrideAngularDamping=true
    bUseGravityOverride=true
    simulationSpace=RootRelative
    alphaBoolBlend {blendOption=Linear}
    boundBone {boneName=Dyn_Hair_Side_05}
    componentAppliedLinearAccClamp=<500 500 500>
    componentLinearAccScale=<0.7 0.7 0.7>
    componentLinearVelScale=<0.7 0.7 0.7>
    gravityOverride {z=-980}
    physicsBodyDefinitions [{
      boundBone {boneName=Dyn_Hair_Side_05}
      boxExtents=<10 10 10>
      constraintSetup {angularLimitsMax=<0 0 7>, angularLimitsMin=<0 -20 -7>, angularTarget=<0 0 0>}
      localJointOffset=<0.5 0 0>
     }]
   }
   Anim Dynamics - Bone: Dyn_Hair_Side_04(ComponentPose=@n5, Alpha=1, ExternalForce="0,0,0") {
    angularDampingOverride=0.9999
    angularSpringConstant=2000
    bAngularSpring=true
    bOverrideAngularDamping=true
    bUseGravityOverride=true
    simulationSpace=RootRelative
    alphaBoolBlend {blendOption=Linear}
    boundBone {boneName=Dyn_Hair_Side_04}
    componentAppliedLinearAccClamp=<500 500 500>
    componentLinearAccScale=<0.7 0.7 0.7>
    componentLinearVelScale=<0.7 0.7 0.7>
    gravityOverride {z=-980}
    physicsBodyDefinitions [{
      boundBone {boneName=Dyn_Hair_Side_04}
      boxExtents=<10 10 10>
      constraintSetup {angularLimitsMax=<0 0 7>, angularLimitsMin=<0 -20 -7>, angularTarget=<0 0 0>}
      localJointOffset=<0.5 0 0>
     }]
   }
   n6 = Copy Pose From Mesh(SourceMeshComponent=Get Body() -> (Body:Skeletal Mesh Component Object Reference)) {bUseAttachedParent=true}
   n7 = Copy Pose From Mesh() {bUseAttachedParent=true}
   n8 = RigidBody(ComponentPose=Local To Component(LocalPose=Layered blend per bone(BasePose=@n7, BlendPoses_0=@n6) {layerSetup [{branchFilters [{blendDepth=2, boneName=clavicle_r}, {blendDepth=2, boneName=clavicle_l}]}]})) {
    bClampLinearTranslationLimitToRefPose=true
    bOverrideWorldGravity=true
    overridePhysicsAsset=&/F_MED_BugBeliever_FaceAcc_Physics
    simulationSpace=BaseBoneSpace
    alphaBoolBlend {blendOption=Linear}
    baseBoneRef {boneName=head}
    overrideWorldGravity {z=-980}
    simSpaceSettings {
     dampingAlpha=0
     maxAngularAcceleration=300
     maxAngularVelocity=2
     maxLinearAcceleration=400
     maxLinearVelocity=250
     worldAlpha=1
     externalLinearDragV=<0.1 0.1>
    }
   }
   n9 = Anim Dynamics - Bone: Dyn_Hair_Bakc_01(ComponentPose=@n8, Alpha=1, ExternalForce="0,0,0") {
    angularDampingOverride=0.9999
    angularSpringConstant=2000
    bAngularSpring=true
    bOverrideAngularDamping=true
    bUseGravityOverride=true
    simulationSpace=RootRelative
    alphaBoolBlend {blendOption=Linear}
    boundBone {boneName=Dyn_Hair_Bakc_01}
    componentAppliedLinearAccClamp=<500 500 500>
    componentLinearAccScale=<0.7 0.7 0.7>
    componentLinearVelScale=<0.7 0.7 0.7>
    gravityOverride {z=-980}
    physicsBodyDefinitions [{
      boundBone {boneName=Dyn_Hair_Bakc_01}
      boxExtents=<10 10 10>
      constraintSetup {angularLimitsMax=<0 0 7>, angularLimitsMin=<0 -20 -7>, angularTarget=<0 0 0>}
      localJointOffset=<0.5 0 0>
     }]
   }
   n10 = Anim Dynamics - Bone: Dyn_Hair_Bakc_02(ComponentPose=@n9, Alpha=1, ExternalForce="0,0,0") {
    angularDampingOverride=0.9999
    angularSpringConstant=1500
    bAngularSpring=true
    bOverrideAngularDamping=true
    bUseGravityOverride=true
    simulationSpace=RootRelative
    alphaBoolBlend {blendOption=Linear}
    boundBone {boneName=Dyn_Hair_Bakc_02}
    componentAppliedLinearAccClamp=<500 500 500>
    componentLinearAccScale=<0.7 0.7 0.7>
    componentLinearVelScale=<0.7 0.7 0.7>
    gravityOverride {z=-980}
    physicsBodyDefinitions [{
      boundBone {boneName=Dyn_Hair_Bakc_02}
      boxExtents=<10 10 10>
      constraintSetup {angularLimitsMax=<5 5 7>, angularLimitsMin=<-5 -8 -7>, angularTarget=<0 0 0>}
      localJointOffset=<1 0 0>
     }]
   }
   n11 = Anim Dynamics - Bone: Dyn_Hair_Bakc_03(ComponentPose=@n10, Alpha=1, ExternalForce="0,0,0") {
    angularDampingOverride=0.9999
    angularSpringConstant=1500
    bAngularSpring=true
    bOverrideAngularDamping=true
    bUseGravityOverride=true
    simulationSpace=RootRelative
    alphaBoolBlend {blendOption=Linear}
    boundBone {boneName=Dyn_Hair_Bakc_03}
    componentAppliedLinearAccClamp=<500 500 500>
    componentLinearAccScale=<0.7 0.7 0.7>
    componentLinearVelScale=<0.7 0.7 0.7>
    gravityOverride {z=-980}
    physicsBodyDefinitions [{
      boundBone {boneName=Dyn_Hair_Bakc_03}
      boxExtents=<10 10 10>
      constraintSetup {angularLimitsMax=<5 5 7>, angularLimitsMin=<-5 -8 -7>, angularTarget=<0 0 0>}
      localJointOffset=<1 0 0>
     }]
   }
   n12 = Anim Dynamics - Bone: Dyn_Hair_Bakc_04(ComponentPose=@n11, Alpha=1, ExternalForce="0,0,0") {
    angularDampingOverride=0.9999
    angularSpringConstant=1500
    bAngularSpring=true
    bOverrideAngularDamping=true
    bUseGravityOverride=true
    simulationSpace=RootRelative
    alphaBoolBlend {blendOption=Linear}
    boundBone {boneName=Dyn_Hair_Bakc_04}
    componentAppliedLinearAccClamp=<500 500 500>
    componentLinearAccScale=<0.7 0.7 0.7>
    componentLinearVelScale=<0.7 0.7 0.7>
    gravityOverride {z=-980}
    physicsBodyDefinitions [{
      boundBone {boneName=Dyn_Hair_Bakc_04}
      boxExtents=<10 10 10>
      constraintSetup {angularLimitsMax=<5 -8 7>, angularLimitsMin=<-5 0 -7>, angularTarget=<0 0 0>}
      localJointOffset=<1 0 0>
     }]
   }
   n13 = Anim Dynamics - Bone: Dyn_Hair_Bakc_05(ComponentPose=@n12, Alpha=1, ExternalForce="0,0,0") {
    angularDampingOverride=0.9999
    angularSpringConstant=1500
    bAngularSpring=true
    bOverrideAngularDamping=true
    bUseGravityOverride=true
    simulationSpace=RootRelative
    alphaBoolBlend {blendOption=Linear}
    boundBone {boneName=Dyn_Hair_Bakc_05}
    componentAppliedLinearAccClamp=<500 500 500>
    componentLinearAccScale=<0.7 0.7 0.7>
    componentLinearVelScale=<0.7 0.7 0.7>
    gravityOverride {z=-980}
    physicsBodyDefinitions [{
      boundBone {boneName=Dyn_Hair_Bakc_05}
      boxExtents=<10 10 10>
      constraintSetup {angularLimitsMax=<5 5 7>, angularLimitsMin=<-5 -8 -7>, angularTarget=<0 0 0>}
      localJointOffset=<1 0 0>
     }]
   }
   n14 = Anim Dynamics - Bone: Dyn_Hair_Bakc_06(ComponentPose=@n13, Alpha=1, ExternalForce="0,0,0") {
    angularDampingOverride=0.9999
    angularSpringConstant=1500
    bAngularSpring=true
    bOverrideAngularDamping=true
    bUseGravityOverride=true
    simulationSpace=RootRelative
    alphaBoolBlend {blendOption=Linear}
    boundBone {boneName=Dyn_Hair_Bakc_06}
    componentAppliedLinearAccClamp=<500 500 500>
    componentLinearAccScale=<0.7 0.7 0.7>
    componentLinearVelScale=<0.7 0.7 0.7>
    gravityOverride {z=-980}
    physicsBodyDefinitions [{
      boundBone {boneName=Dyn_Hair_Bakc_06}
      boxExtents=<10 10 10>
      constraintSetup {angularLimitsMax=<5 5 7>, angularLimitsMin=<-5 -8 -7>, angularTarget=<0 0 0>}
      localJointOffset=<1 0 0>
     }]
   }
   n15 = Anim Dynamics - Bone: Dyn_Hair_Bakc_07(ComponentPose=@n14, Alpha=1, ExternalForce="0,0,0") {
    angularDampingOverride=0.9999
    angularSpringConstant=1500
    bAngularSpring=true
    bOverrideAngularDamping=true
    bUseGravityOverride=true
    simulationSpace=RootRelative
    alphaBoolBlend {blendOption=Linear}
    boundBone {boneName=Dyn_Hair_Bakc_07}
    componentAppliedLinearAccClamp=<500 500 500>
    componentLinearAccScale=<0.7 0.7 0.7>
    componentLinearVelScale=<0.7 0.7 0.7>
    gravityOverride {z=-980}
    physicsBodyDefinitions [{
      boundBone {boneName=Dyn_Hair_Bakc_07}
      boxExtents=<10 10 10>
      constraintSetup {angularLimitsMax=<5 5 7>, angularLimitsMin=<-5 -8 -7>, angularTarget=<0 0 0>}
      localJointOffset=<1 0 0>
     }]
   }
   n16 = Anim Dynamics - Bone: Dyn_Hair_Fr_L(ComponentPose=@n15, Alpha=1, ExternalForce="0,0,0") {
    angularDampingOverride=0.9999
    angularSpringConstant=500
    bAngularSpring=true
    bOverrideAngularDamping=true
    bUseGravityOverride=true
    simulationSpace=RootRelative
    alphaBoolBlend {blendOption=Linear}
    boundBone {boneName=Dyn_Hair_Fr_L}
    componentAppliedLinearAccClamp=<500 500 500>
    componentLinearAccScale=<1 1 1>
    componentLinearVelScale=<1 1 1>
    gravityOverride {z=-980}
    physicsBodyDefinitions [{
      boundBone {boneName=Dyn_Hair_Fr_L}
      boxExtents=<10 10 10>
      constraintSetup {angularLimitsMax=<0 0 5>, angularLimitsMin=<0 -10 -5>, angularTarget=<0 0 0>}
      localJointOffset=<1 0 0>
     }]
   }
   n17 = Anim Dynamics - Bone: Dyn_Hair_Fr_R(ComponentPose=@n16, Alpha=1, ExternalForce="0,0,0") {
    angularDampingOverride=0.9999
    angularSpringConstant=500
    bAngularSpring=true
    bOverrideAngularDamping=true
    bUseGravityOverride=true
    simulationSpace=RootRelative
    alphaBoolBlend {blendOption=Linear}
    boundBone {boneName=Dyn_Hair_Fr_R}
    componentAppliedLinearAccClamp=<500 500 500>
    componentLinearAccScale=<1 1 1>
    componentLinearVelScale=<1 1 1>
    gravityOverride {z=-980}
    physicsBodyDefinitions [{
      boundBone {boneName=Dyn_Hair_Fr_R}
      boxExtents=<10 10 10>
      constraintSetup {angularLimitsMax=<0 0 5>, angularLimitsMin=<0 -10 -5>, angularTarget=<0 0 0>}
      localJointOffset=<1 0 0>
     }]
   }
   n18 = Anim Dynamics - Bone: Dyn_earring_L(ComponentPose=@n17, Alpha=1, ExternalForce="0,0,0") {
    angularDampingOverride=0.9999
    angularSpringConstant=500
    bAngularSpring=true
    bOverrideAngularDamping=true
    bUseGravityOverride=true
    simulationSpace=RootRelative
    alphaBoolBlend {blendOption=Linear}
    boundBone {boneName=Dyn_earring_L}
    componentAppliedLinearAccClamp=<500 500 500>
    componentLinearAccScale=<1 1 1>
    componentLinearVelScale=<1 1 1>
    gravityOverride {z=-980}
    physicsBodyDefinitions [{
      boundBone {boneName=Dyn_earring_L}
      boxExtents=<10 10 10>
      constraintSetup {angularLimitsMax=<5 5 5>, angularLimitsMin=<-5 -50 -10>, angularTarget=<0 0 0>}
      localJointOffset=<1 0 0>
     }]
   }
   n19 = Anim Dynamics - Bone: Dyn_earring_R(ComponentPose=@n18, Alpha=1, ExternalForce="0,0,0") {
    angularDampingOverride=0.9999
    angularSpringConstant=500
    bAngularSpring=true
    bOverrideAngularDamping=true
    bUseGravityOverride=true
    simulationSpace=RootRelative
    alphaBoolBlend {blendOption=Linear}
    boundBone {boneName=Dyn_earring_R}
    componentAppliedLinearAccClamp=<500 500 500>
    componentLinearAccScale=<1 1 1>
    componentLinearVelScale=<1 1 1>
    gravityOverride {z=-980}
    physicsBodyDefinitions [{
      boundBone {boneName=Dyn_earring_R}
      boxExtents=<10 10 10>
      constraintSetup {angularLimitsMax=<5 5 5>, angularLimitsMin=<-5 -50 -10>, angularTarget=<0 0 0>}
      localJointOffset=<1 0 0>
     }]
   }
   Output Pose(Result=Component To Local(ComponentPose=@n19)) {layerGroup=None}
  ]}
 EventGraph {nodes [
   n1 = Event OnCharacterCustomizationLoaded()
   > Set Body(Body:Skeletal Mesh Component Object Reference=Get Array Item with EFort Custom Part Type(Array=@n1.CustomMeshParts)) -> (Output_Get:Skeletal Mesh Component Object Reference)
   Event BlueprintUpdateAnimation()
  ]}
]
variables [
 Body {type=SkeletalMeshComponent}
 Gravity Override Params {type=GravityOverrideParamsStruct}
]
